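# Map screen widget that labels and shows how many players (and squads) remain in an Athena match, built on AthenaMapPlayersLeftWidget.
# A vertical box stacks an uppercase blue title over the count text inherited from the parent class; the graph never sets that count itself.
# When the parent fires OnMaxSquadSizeSet, the title switches between "Players Left" and "Players (Squads) Left" based on whether max squad size exceeds one.
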WidgetBlueprint /FortniteCoreUI/HUD/MiniMap/MapScreen/WBP_Athena_Map_PlayersLeft
graphs [EventGraph {nodes [
   n1 = Event OnMaxSquadSizeSet!() -> (OutputDelegate, then, MaxSquadSize:int=0) {is=title K2Node_Event}
   > "SetText (Text)"(
    self=Get Text_Title() -> (Text_Title:Common Text Object Reference)
    InText=Select(
     Option 0:Text=Get PlayersLeft() -> (PlayersLeft:Text)
     Option 1:Text=Get PlayersSquadsLeft() -> (PlayersSquadsLeft:Text)
     Index="integer > integer"(A=@n1.MaxSquadSize, B=1) {is=title K2Node_CallFunction}
    ) -> (ReturnValue:Text)
   )
  ]}]
variables [
 PlayersLeft {type=string, default=Players Left}
 PlayersSquadsLeft {type=string, default="Players (Squads) Left"}
]
widgets {
 info {inheritedWidgetCount=2, parentClass=/Script/FortniteUI.AthenaMapPlayersLeftWidget, rootWidgetClass=&VerticalBox, widgetCount=4}
 widgets [
  VerticalBox_31 {widgetClassPath=&VerticalBox, widget=VerticalBox_31}
  Text_Title {
   widgetClassPath=&CommonTextBlock
   bIsVariable=true
   parent=VerticalBox_31
   widget=Text_Title
   properties {
    text="{PLAYERS (SQUADS) LEFT}"
    textTransformPolicy=ToUpper
    colorAndOpacity {colorUseRule=UseColor_Specified, specifiedColor=<rgba 0.2122 0.6514 1 1>}
    font {
     bForceMonospaced=false
     bMaterialIsStencil=false
     fontMaterial=null
     fontObject=&BurbankSmall
     letterSpacing=0
     monospacedWidth=1
     size=18
     skewAmount=0
     typefaceFontName=Bold
     outlineSettings {
      bApplyOutlineToDropShadows=false
      bMiteredCorners=false
      bSeparateFillAlpha=false
      outlineMaterial=null
      outlineSize=0
      outlineColor=<rgba 0 0 0 1>
     }
    }
    shadowOffset=<0 0>
   }
   slot {class=VerticalBoxSlot, properties {padding=<ltrb 0 0 0 8>}}
  }
  Text_PlayersTeamsLeft {
   widgetClassPath=&CommonTextBlock
   bInherited=true
   parent=VerticalBox_31
   widget=Text_PlayersTeamsLeft
   properties {
    text="{0 (0)}"
    font {
     bForceMonospaced=false
     bMaterialIsStencil=false
     fontMaterial=null
     fontObject=&BurbankSmall
     letterSpacing=0
     monospacedWidth=1
     size=16.5
     skewAmount=0
     typefaceFontName=Bold
     outlineSettings {
      bApplyOutlineToDropShadows=false
      bMiteredCorners=false
      bSeparateFillAlpha=false
      outlineMaterial=null
      outlineSize=0
      outlineColor=<rgba 0 0 0 1>
     }
    }
    shadowOffset=<0 0>
   }
   slot {class=VerticalBoxSlot, properties {horizontalAlignment=HAlign_Left}}
  }
  TutorialHighlight {widgetClassPath=&TutorialHighlight}
 ]
}
