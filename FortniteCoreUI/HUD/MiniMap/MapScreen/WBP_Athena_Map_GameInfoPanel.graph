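# Full-screen map game info panel, based on AthenaMapGameInfo, showing the playlist name, storm info, squad eliminations and players left.
# It registers as a player-specific widget host on Construct and unregisters on Destruct.
# Content matching its MapGameInfo tag exactly is added to or removed from VBox_Content, ignoring requested padding and fill.
# Extension widgets for the squad-eliminations slot fill Overlay_SquadElims, while others append to VBox_Content.
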WidgetBlueprint /FortniteCoreUI/HUD/MiniMap/MapScreen/WBP_Athena_Map_GameInfoPanel
graphs [
 AddContentToHostWidget {nodes [
   n1 = AddContentToHostWidget!() -> (then, HostSlotTag:Gameplay Tag Structure, ContentWidget:Widget Object Reference, Padding:Margin Structure, bFill:bool=~) {is=title K2Node_FunctionEntry}
   > Branch(Condition=MatchesTag!(
     TagOne:Gameplay Tag Structure=@n1
     TagTwo:Gameplay Tag Structure=Get ContentHostTag() -> (ContentHostTag:Gameplay Tag Structure)
     bExactMatch:bool=true
    ) -> (ReturnValue:bool))
   > AddChildToVerticalBox(
    self=Get VBox_Content() -> (VBox_Content:Vertical Box Object Reference)
    Content=Get ContentWidget() -> (ContentWidget:Widget Object Reference)
   )
  ]}
 EventGraph {nodes [
   Event Construct()
   > RegisterHostSlots()
   Event Destruct()
   > UnregisterHostSlots()
   n1 = Event Add Content To Host Widget!() -> (
    OutputDelegate
    then
    WidgetHostTag:"Gameplay Tag Structure (by ref)"
    ContentWidget:Widget Object Reference
    Padding:"Margin Structure (by ref)"
    bFill:bool
   ) {is=title K2Node_Event}
   > AddContentToHostWidget!(
    execute
    self
    HostSlotTag:Gameplay Tag Structure=@n1.WidgetHostTag
    ContentWidget:Widget Object Reference=@n1.ContentWidget
    Padding:Margin Structure=@n1.Padding
    bFill:bool=@n1.bFill
   ) -> (then) {is=title K2Node_CallFunction}
   n2 = Event OnExtensionWidgetCreated!() -> (OutputDelegate, then, ExtensionSlot:EUIExtensionSlot Enum=Invalid, ExtensionWidget:User Widget Object Reference) {is=title K2Node_Event}
   > n3 = Branch(Condition="Equal (Enum)"(A:EUIExtensionSlot Enum=@n2.ExtensionSlot, B:EUIExtensionSlot Enum=FullScreenMapSquadEliminationsOverlay))
   > n4 = AddChildToOverlay!(
    execute
    self:Overlay Object Reference=Get Overlay_SquadElims() -> (Overlay_SquadElims:Overlay Object Reference)
    Content:Widget Object Reference=@n2.ExtensionWidget
   ) -> (then, ReturnValue:Overlay Slot Object Reference)
   > SetHorizontalAlignment(self=@n4, InHorizontalAlignment=HAlign_Fill)
   > SetVerticalAlignment(self=@n4, InVerticalAlignment=VAlign_Fill)
   AddChildToVerticalBox(execute=@n3.else, self=Get VBox_Content(), Content=@n2.ExtensionWidget)
  ]}
 RegisterHostSlots {nodes [
   RegisterHostSlots() {is=title K2Node_FunctionEntry}
   > RegisterWidgetHost!(
    execute
    WidgetHost:Widget Object Reference
    WidgetHostTag:"Gameplay Tag Structure (by ref)"=Get ContentHostTag()
    IsPlayerSpecific:bool=true
   ) -> (then)
  ]}
 RemoveContentFromWidgetHost {nodes [
   n1 = Remove Content From Host Widget!() -> (
    then
    WidgetHostTag:"Gameplay Tag Structure (by ref)"
    ContentWidget:Widget Object Reference
    RemovalType:EWidgetSlottingRemovalType Enum=Unparent
   ) {is=title K2Node_FunctionEntry}
   > n2 = Branch(Condition=MatchesTag(TagOne=@n1, TagTwo=Get ContentHostTag(), bExactMatch=true))
   > n3 = RemoveChild(self=Get VBox_Content(), Content=Get ContentWidget())
   > Return Node!(execute, ReturnValue:bool=@n3)
   Return Node(execute=@n2.else, ReturnValue=false)
  ]}
 UnregisterHostSlots {nodes [
   UnregisterHostSlots() {is=title K2Node_FunctionEntry}
   > UnregisterWidgetHost!(execute, WidgetHost:Widget Object Reference, WidgetHostTag:"Gameplay Tag Structure (by ref)"=Get ContentHostTag()) -> (then)
  ]}
]
variables [ContentHostTag {type=GameplayTag, default {tagName=DynamicUI.HUD.Fort.Athena.MapScreen.MapGameInfo.Content}}]
widgets {
 info {inheritedWidgetCount=2, parentClass=/Script/FortniteUI.AthenaMapGameInfo, rootWidgetClass=&SizeBox, widgetCount=11}
 widgets [
  SizeBox_0 {
   widgetClassPath=&SizeBox
   widget=SizeBox_0
   properties {bOverride_MaxDesiredWidth=true, maxDesiredWidth=384, minDesiredWidth=384, widthOverride=384}
  }
  VerticalBox_0 {widgetClassPath=&VerticalBox, parent=SizeBox_0, widget=VerticalBox_0, slot {class=SizeBoxSlot}}
  Border_Title {
   widgetClassPath=&CommonBorder
   parent=VerticalBox_0
   widget=Border_Title
   properties {
    verticalAlignment=VAlign_Center
    background {
     bIsDynamicallyLoaded=false
     drawAs=Box
     imageType=NoImage
     mirroring=NoMirror
     resourceName=None
     resourceObject=null
     tiling=NoTile
     imageSize=<32 32>
     margin=<ltrb 0 0 0 0>
     outlineSettings {
      bUseBrushTransparency=false
      roundingType=HalfHeightRadius
      width=0
      color {colorUseRule=UseColor_Specified, specifiedColor=<rgba 0 0 0 0>}
      cornerRadii=<0 0 0 0>
     }
     tintColor {colorUseRule=UseColor_Specified, specifiedColor=<rgba 0.0086 0.1529 0.7913 1>}
     uVRegion {bIsValid=false, max=<0 0>, min=<0 0>}
    }
    brushColor=<rgba 1 1 1 0.5>
    padding=<ltrb 32 8 32 8>
   }
   slot {class=VerticalBoxSlot}
  }
  Text_PlaylistName {
   widgetClassPath=&CommonTextBlock
   bInherited=true
   parent=Border_Title
   widget=Text_PlaylistName
   properties {
    text="{Solo}"
    textTransformPolicy=ToUpper
    font {
     bForceMonospaced=false
     bMaterialIsStencil=false
     fontMaterial=null
     fontObject=&Burbank
     letterSpacing=0
     monospacedWidth=1
     size=19.5
     skewAmount=0
     typefaceFontName=Burbank-PlayerPower
     outlineSettings {
      bApplyOutlineToDropShadows=false
      bMiteredCorners=false
      bSeparateFillAlpha=false
      outlineMaterial=null
      outlineSize=0
      outlineColor=<rgba 0 0 0 1>
     }
    }
    renderTransform {angle=0, scale=<1 1>, shear=<-12 0>, translation=<0 0>}
    shadowOffset=<0 0>
   }
   slot {class=BorderSlot, properties {verticalAlignment=VAlign_Center, padding=<ltrb 32 8 32 8>}}
  }
  Border_Content {
   widgetClassPath=&CommonBorder
   parent=VerticalBox_0
   widget=Border_Content
   properties {
    background {
     bIsDynamicallyLoaded=false
     drawAs=Image
     imageType=NoImage
     mirroring=NoMirror
     resourceName=None
     resourceObject=null
     tiling=NoTile
     imageSize=<32 32>
     margin=<ltrb 0 0 0 0>
     outlineSettings {
      bUseBrushTransparency=false
      roundingType=HalfHeightRadius
      width=0
      color {colorUseRule=UseColor_Specified, specifiedColor=<rgba 0 0 0 0>}
      cornerRadii=<0 0 0 0>
     }
     tintColor {colorUseRule=UseColor_Specified, specifiedColor=<rgba 1 1 1 1>}
     uVRegion {bIsValid=false, max=<0 0>, min=<0 0>}
    }
    brushColor=<rgba 0 0.0005 0.025 0.45>
    padding=<ltrb 24 12 24 12>
   }
   slot {class=VerticalBoxSlot}
  }
  VBox_Content {
   widgetClassPath=&VerticalBox
   bIsVariable=true
   parent=Border_Content
   widget=VBox_Content
   slot {class=BorderSlot, properties {padding=<ltrb 24 12 24 12>}}
  }
  WBP_Athena_Map_StormInfo {
   widgetClassPath=&/WBP_Athena_Map_StormInfo
   bIsVariable=true
   parent=VBox_Content
   widget=WBP_Athena_Map_StormInfo
   slot {class=VerticalBoxSlot}
  }
  Overlay_SquadElims {
   widgetClassPath=&Overlay
   bIsVariable=true
   parent=VBox_Content
   widget=Overlay_SquadElims
   slot {class=VerticalBoxSlot, properties {padding=<ltrb 0 12 0 0>}}
  }
  WBP_SquadElims {
   widgetClassPath=&/WBP_Athena_Map_SquadElims
   bIsVariable=true
   parent=Overlay_SquadElims
   widget=WBP_SquadElims
   slot {class=OverlaySlot, properties {horizontalAlignment=HAlign_Fill, verticalAlignment=VAlign_Fill}}
  }
  WBP_PlayersLeft {
   widgetClassPath=&/WBP_Athena_Map_PlayersLeft
   bIsVariable=true
   parent=VBox_Content
   widget=WBP_PlayersLeft
   slot {class=VerticalBoxSlot, properties {padding=<ltrb 0 12 0 0>}}
  }
  TutorialHighlight {widgetClassPath=&TutorialHighlight}
 ]
}
