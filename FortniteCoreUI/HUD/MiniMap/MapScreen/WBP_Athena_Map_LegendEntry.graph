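# Widget Blueprint for a single legend entry on the Athena map screen, based on the native AthenaMapLegendEntryWidget class.
# A horizontal row holds a left-aligned DynamicEntryBox that generates WBP_LegendIcon widgets, followed by a right-aligned name label filling the remaining width.
# Both the icon box and the text are inherited bindings, and the event graph has only empty construct events, so the Blueprint supplies just layout and styling.
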WidgetBlueprint /FortniteCoreUI/HUD/MiniMap/MapScreen/WBP_Athena_Map_LegendEntry
graphs [EventGraph {nodes [
   Event PreConstruct()
   Event Construct()
  ]}]
variables []
widgets {
 info {inheritedWidgetCount=2, parentClass=/Script/FortniteUI.AthenaMapLegendEntryWidget, rootWidgetClass=&HorizontalBox, widgetCount=3}
 widgets [
  HorizontalBox_50 {widgetClassPath=&HorizontalBox, widget=HorizontalBox_50}
  DynamicEntryBox_LegendIcons {
   widgetClassPath=&DynamicEntryBox
   bInherited=true
   bIsVariable=true
   parent=HorizontalBox_50
   widget=DynamicEntryBox_LegendIcons
   properties {entryWidgetClass=/FortniteCoreUI/HUD/MiniMap/WBP_LegendIcon.WBP_LegendIcon_C, numDesignerPreviewEntries=1}
   slot {class=HorizontalBoxSlot, properties {horizontalAlignment=HAlign_Left}}
  }
  Text_EntryName {
   widgetClassPath=&CommonTextBlock
   bInherited=true
   bIsVariable=true
   parent=HorizontalBox_50
   widget=Text_EntryName
   properties {
    justification=Right
    text=$Legend Entry
    font {
     bForceMonospaced=false
     bMaterialIsStencil=false
     fontMaterial=null
     fontObject=&BurbankSmall
     letterSpacing=0
     monospacedWidth=1
     size=16.5
     skewAmount=0
     typefaceFontName=Bold
     outlineSettings {
      bApplyOutlineToDropShadows=false
      bMiteredCorners=false
      bSeparateFillAlpha=false
      outlineMaterial=null
      outlineSize=0
      outlineColor=<rgba 0 0 0 1>
     }
    }
    shadowOffset=<0 0>
   }
   slot {
    class=HorizontalBoxSlot
    properties {horizontalAlignment=HAlign_Right, verticalAlignment=VAlign_Center, padding=<ltrb 8 0 0 0>, size {sizeRule=Fill, value=1}}
   }
  }
 ]
}
