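# UI-domain material for a game info portrait in the Athena HUD, compositing a hero image over a tinted gradient.
# The RGB parameter forms the background, blended toward double brightness along the V coordinate.
# The Portrait texture is sampled with UVs scaled about the center by the Scale parameter.
# Its RGB serves as both color and per-channel blend factor, so dark areas reveal the gradient instead of using texture alpha.
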Material /FortniteCoreUI/Athena/HUD/Materials/M_UI_GameInfo_Portait
graph {
 properties {bUsedWithStaticMesh=false, materialDomain=MD_UI, pixelDepthOffsetMode=PDOM_Legacy, usedShadingModels=Default Lit}
 nodes [
  n1 = VectorParameter() {parameterName=RGB, defaultValue=<rgba 0 0.2433 1 1>}
  n2 = TextureCoordinate()
  n3 = TextureSampleParameter2D(UVs=MaterialFunctionCall(UVs=TextureCoordinate(), Texture Scale=ScalarParameter() {defaultValue=1.5, parameterName=Scale}) {materialFunction=&ScaleUVsByCenter}.UVs) {parameterName=Portrait, texture=/Game/UI/Foundation/Textures/Icons/Heroes/Athena/Soldier/T-Soldier-HID-005-Athena-Commando-M}
  LinearInterpolate(A=LinearInterpolate(A=@n1.RGB, B=Multiply(A=@n1.RGB) {constB=2}, Alpha=ComponentMask(None=@n2) {g=true}), B=@n3.RGB, Alpha=@n3.RGB)
 ]
}
