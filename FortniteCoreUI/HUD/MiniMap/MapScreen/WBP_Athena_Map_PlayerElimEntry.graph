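# Row widget for the map screen's player eliminations display, showing portrait, ranked (Habanero) icon, player name and elimination count.
# On a ranked progress update it hides the rank icon, resolves tier data from RankData and RankInfoOverrides, and shows the small icon only if the lookup succeeds.
# ResetRankedProgress collapses that icon.
# Construct, and PreConstruct at design time, color both texts according to bIsLocalPlayer.
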WidgetBlueprint /FortniteCoreUI/HUD/MiniMap/MapScreen/WBP_Athena_Map_PlayerElimEntry
graphs [EventGraph {nodes [
   n1 = Event UpdateRankedProgress!() -> (OutputDelegate, then, RankedProgress:"Ranked Progress Replicated Data Structure (by ref)") {is=title K2Node_Event}
   > SetVisibility(self=Get LazyImage_Habanero() -> (LazyImage_Habanero:Fort Lazy Image Object Reference), InVisibility=Collapsed)
   n2 = Break Ranked Progress Replicated Data!(RankedProgressReplicatedData:"Ranked Progress Replicated Data Structure (by ref)"=@n1.RankedProgress) -> (RankType:String, Rank:int, HighestRank:int, CurrentMMR:int, CurrentPosition:int) {is=title K2Node_BreakStruct}
   n3 = GetHabaneroDisplayDataOverride!(
    InDisplayData:Fort Habanero Display Data Object Reference=Get RankData() -> (RankData:Fort Habanero Display Data Object Reference)
    InOverrideData:Fort Compete UIHabanero Tier Override List Object Reference=Get RankInfoOverrides() -> (RankInfoOverrides:Fort Compete UIHabanero Tier Override List Object Reference)
    InKey:"Fort Compete UIHabanero Tier Override Key Structure (by ref)"=Make Fort Compete UIHabanero Tier Override Key!(
     RankedTrack:Name=String To Name(InString=@n2)
     Season:String
     Rank:int=@n2.Rank
     MatchmakingRating:int=@n2.CurrentMMR
     CurrentPosition:int=@n2.CurrentPosition
    ) -> (FortCompeteUIHabaneroTierOverrideKey:Fort Compete UIHabanero Tier Override Key Structure) {is=title K2Node_MakeStruct}
   ) -> (OutData:Fort Habanero Tier Override Structure, ReturnValue:bool)
   > Branch(Condition=@n3.ReturnValue)
   n4 = Get LazyImage_Habanero()
   > SetBrushFromSoftTexture!(
    execute
    self:Image Object Reference=@n4
    SoftTexture:Texture 2D Soft Object Reference=Break Fort Habanero Tier!(FortHabaneroTier:"Fort Habanero Tier Structure (by ref)"=@n3) -> (
     ForegroundIcon:Material Instance Soft Object Reference
     ForegroundIconSmall:Material Instance Soft Object Reference
     BackgroundIcon:Material Instance Soft Object Reference
     BackgroundIconSmall:Material Instance Soft Object Reference
     WhiteIcon:Texture 2D Soft Object Reference
     SimpleIconSmall:Texture 2D Soft Object Reference
     Name:Text
     TierType:Fort Habanero Tier Type Data Object Reference
     AdditionalLeftIcon:Material Instance Soft Object Reference
     AdditionalRightIcon:Material Instance Soft Object Reference
    ) {is=title K2Node_BreakStruct}.SimpleIconSmall
    bMatchSize:bool=false
   ) -> (then)
   > SetVisibility(self=@n4, InVisibility=HitTestInvisible)
   n5 = Event Construct()
   SetColorAndOpacity(
    execute=[@n5.then, @n7.then]
    self=[
     Get Text_PlayerName() -> (Text_PlayerName:Common Text Object Reference)
     Get Text_ElimCount() -> (Text_ElimCount:Common Text Object Reference)
    ]
    InColorAndOpacity=Select(
     Option 0:Slate Color Structure=Make Slate Color(SpecifiedColor=(R=0.4508,G=0.9911,B=1,A=1))
     Option 1:Slate Color Structure=Make Slate Color(SpecifiedColor=(R=1,G=1,B=1,A=1))
     Index=Get bIsLocalPlayer() -> (bIsLocalPlayer:bool)
    ) -> (ReturnValue:Slate Color Structure)
   )
   Event ResetRankedProgress() {is=title K2Node_Event}
   > SetVisibility(self=Get LazyImage_Habanero(), InVisibility=Collapsed)
   n6 = Event PreConstruct()
   > n7 = Branch(Condition=@n6.IsDesignTime)
  ]}]
variables [
 RankData {type=FortHabaneroDisplayData, default=/Game/UI/Habanero/HabaneroDisplayData}
 RankInfoOverrides {type=FortCompeteUIHabaneroTierOverrideList, default=/Game/UI/Habanero/DA_RankDisplayDataOverride}
]
widgets {
 info {
  inheritedWidgetCount=3
  parentClass=/Script/FortniteUI.AthenaMapPlayerEliminationsWidget
  rootWidgetClass=&HorizontalBox
  widgetCount=5
 }
 widgets [
  HorizontalBox_50 {widgetClassPath=&HorizontalBox, widget=HorizontalBox_50}
  Image_PlayerIcon {
   widgetClassPath=&Image
   bInherited=true
   parent=HorizontalBox_50
   widget=Image_PlayerIcon
   properties {
    visibility=SelfHitTestInvisible
    brush {
     bIsDynamicallyLoaded=false
     drawAs=Image
     imageType=NoImage
     mirroring=NoMirror
     resourceName=None
     resourceObject=/FortniteCoreUI/Athena/HUD/Materials/M_UI_GameInfo_Portait
     tiling=NoTile
     imageSize=<32 32>
     margin=<ltrb 0 0 0 0>
     outlineSettings {
      bUseBrushTransparency=false
      roundingType=HalfHeightRadius
      width=0
      color {colorUseRule=UseColor_Specified, specifiedColor=<rgba 0 0 0 0>}
      cornerRadii=<0 0 0 0>
     }
     tintColor {colorUseRule=UseColor_Specified, specifiedColor=<rgba 1 1 1 1>}
     uVRegion {bIsValid=false, max=<0 0>, min=<0 0>}
    }
   }
   slot {class=HorizontalBoxSlot, properties {horizontalAlignment=HAlign_Center, verticalAlignment=VAlign_Center}}
  }
  LazyImage_Habanero {
   widgetClassPath=&FortLazyImage
   bIsVariable=true
   parent=HorizontalBox_50
   widget=LazyImage_Habanero
   properties {renderTransform {angle=0, scale=<1.2 1.2>, shear=<0 0>, translation=<0 0>}}
   slot {class=HorizontalBoxSlot, properties {horizontalAlignment=HAlign_Center, verticalAlignment=VAlign_Center, padding=<ltrb 4 0 4 0>}}
  }
  Text_PlayerName {
   widgetClassPath=&CommonTextBlock
   bInherited=true
   bIsVariable=true
   parent=HorizontalBox_50
   widget=Text_PlayerName
   properties {
    text="{Player Name}"
    font {
     bForceMonospaced=false
     bMaterialIsStencil=false
     fontMaterial=null
     fontObject=&BurbankSmall
     letterSpacing=0
     monospacedWidth=1
     size=16.5
     skewAmount=0
     typefaceFontName=Bold
     outlineSettings {
      bApplyOutlineToDropShadows=false
      bMiteredCorners=false
      bSeparateFillAlpha=false
      outlineMaterial=null
      outlineSize=0
      outlineColor=<rgba 0 0 0 1>
     }
    }
    shadowOffset=<0 0>
   }
   slot {class=HorizontalBoxSlot, properties {verticalAlignment=VAlign_Center, padding=<ltrb 8 0 0 0>, size {sizeRule=Fill, value=1}}}
  }
  Text_ElimCount {
   widgetClassPath=&CommonTextBlock
   bInherited=true
   bIsVariable=true
   parent=HorizontalBox_50
   widget=Text_ElimCount
   properties {
    text=0
    font {
     bForceMonospaced=false
     bMaterialIsStencil=false
     fontMaterial=null
     fontObject=&BurbankSmall
     letterSpacing=0
     monospacedWidth=1
     size=16.5
     skewAmount=0
     typefaceFontName=Bold
     outlineSettings {
      bApplyOutlineToDropShadows=false
      bMiteredCorners=false
      bSeparateFillAlpha=false
      outlineMaterial=null
      outlineSize=0
      outlineColor=<rgba 0 0 0 1>
     }
    }
    shadowOffset=<0 0>
   }
   slot {class=HorizontalBoxSlot, properties {horizontalAlignment=HAlign_Right, verticalAlignment=VAlign_Center}}
  }
 ]
}
