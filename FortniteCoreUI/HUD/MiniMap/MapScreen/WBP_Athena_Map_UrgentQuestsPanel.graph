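# Widget Blueprint for an Urgent Quests panel, filed under the map screen of the Fortnite HUD.
# A fixed-width SizeBox holds a vertical stack with a translucent blue title bar showing sheared, uppercase "Urgent Quests" text.
# Below it, a dark translucent border named for quest entries fills the remaining height but has no child content.
# The event graph has only unconnected PreConstruct, Construct and Tick events, so this file supplies layout only.
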WidgetBlueprint /FortniteCoreUI/HUD/MiniMap/MapScreen/WBP_Athena_Map_UrgentQuestsPanel
graphs [EventGraph {nodes [
   Event PreConstruct()
   Event Construct()
   Event Tick()
  ]}]
variables []
widgets {
 info {inheritedWidgetCount=0, parentClass=&UserWidget, rootWidgetClass=&SizeBox, widgetCount=5}
 widgets [
  SizeBox_0 {widgetClassPath=&SizeBox, widget=SizeBox_0, properties {bOverride_WidthOverride=true, minDesiredWidth=449, widthOverride=450}}
  VerticalBox_5 {widgetClassPath=&VerticalBox, parent=SizeBox_0, widget=VerticalBox_5, slot {class=SizeBoxSlot}}
  Border_Title {
   widgetClassPath=&CommonBorder
   parent=VerticalBox_5
   widget=Border_Title
   properties {
    verticalAlignment=VAlign_Center
    background {
     bIsDynamicallyLoaded=false
     drawAs=Box
     imageType=NoImage
     mirroring=NoMirror
     resourceName=None
     resourceObject=null
     tiling=NoTile
     imageSize=<32 32>
     margin=<ltrb 0 0 0 0>
     outlineSettings {
      bUseBrushTransparency=false
      roundingType=HalfHeightRadius
      width=0
      color {colorUseRule=UseColor_Specified, specifiedColor=<rgba 0 0 0 0>}
      cornerRadii=<0 0 0 0>
     }
     tintColor {colorUseRule=UseColor_Specified, specifiedColor=<rgba 0.0086 0.1529 0.7913 1>}
     uVRegion {bIsValid=false, max=<0 0>, min=<0 0>}
    }
    brushColor=<rgba 1 1 1 0.5>
    padding=<ltrb 32 8 32 8>
   }
   slot {class=VerticalBoxSlot}
  }
  Text_Title {
   widgetClassPath=&CommonTextBlock
   parent=Border_Title
   widget=Text_Title
   properties {
    text=Urgent Quests
    textTransformPolicy=ToUpper
    font {
     bForceMonospaced=false
     bMaterialIsStencil=false
     fontMaterial=null
     fontObject=&Burbank
     letterSpacing=0
     monospacedWidth=1
     size=18
     skewAmount=0
     typefaceFontName=Burbank-PlayerPower
     outlineSettings {
      bApplyOutlineToDropShadows=false
      bMiteredCorners=false
      bSeparateFillAlpha=false
      outlineMaterial=null
      outlineSize=0
      outlineColor=<rgba 0 0 0 1>
     }
    }
    renderTransform {angle=0, scale=<1 1>, shear=<-12 0>, translation=<0 0>}
    shadowOffset=<0 0>
   }
   slot {class=BorderSlot, properties {verticalAlignment=VAlign_Center, padding=<ltrb 32 8 32 8>}}
  }
  Border_QuestEntries {
   widgetClassPath=&CommonBorder
   parent=VerticalBox_5
   widget=Border_QuestEntries
   properties {
    background {
     bIsDynamicallyLoaded=false
     drawAs=Image
     imageType=NoImage
     mirroring=NoMirror
     resourceName=None
     resourceObject=null
     tiling=NoTile
     imageSize=<32 32>
     margin=<ltrb 0 0 0 0>
     outlineSettings {
      bUseBrushTransparency=false
      roundingType=HalfHeightRadius
      width=0
      color {colorUseRule=UseColor_Specified, specifiedColor=<rgba 0 0 0 0>}
      cornerRadii=<0 0 0 0>
     }
     tintColor {colorUseRule=UseColor_Specified, specifiedColor=<rgba 1 1 1 1>}
     uVRegion {bIsValid=false, max=<0 0>, min=<0 0>}
    }
    brushColor=<rgba 0 0.0005 0.025 0.45>
    padding=<ltrb 24 12 24 12>
   }
   slot {class=VerticalBoxSlot, properties {size {sizeRule=Fill, value=1}}}
  }
 ]
}
